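# Ambient outdoor sound cue for medium sail flapping at the Shipwreck POI, mixing two layers identically at every quality level.
# One layer repeatedly plays random, pitch-varied large flap one-shots separated by random delays.
# The other is a delayed flap loop through three nested looping Envelopers, whose pitch and differently timed volume curves keep it evolving.
# Overridden capsule attenuation adds occlusion, distance filtering and listener focus, alongside a distance-based reverb send.
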SoundCue /Game/Sounds/Helios/POI/Shipwreck/Spots/Spot_Shipwreck_Sail_Flapping_Medium_Loop_Cue
root=QualityLevel_1
nodes [
 {id=Delay_0, children [Enveloper_4], properties {delayMax=0.75, delayMin=0.2}}
 {id=Delay_2, children [Random_2], properties {delayMax=6, delayMin=3}}
 {
  id=Enveloper_0
  children [WavePlayer_0]
  properties {
   bLoop=true
   bLoopIndefinitely=true
   loopEnd=60
   pitchCurve {editorCurveData {keys [
      {interpMode=RCIM_Cubic, time=0}
      <at 15 0.85 RCIM_Cubic>
      {arriveTangent=0.0091, interpMode=RCIM_Cubic, leaveTangent=0.0091, tangentMode=RCTM_User, time=30}
      <at 45 1.1 RCIM_Cubic>
      {time=60}
     ]}}
   volumeCurve {editorCurveData {keys [<at 0 1>]}}
  }
 }
 {
  id=Enveloper_3
  children [Enveloper_0]
  properties {
   bLoop=true
   bLoopIndefinitely=true
   loopEnd=60
   pitchCurve {editorCurveData {keys [{time=0}]}}
   volumeCurve {editorCurveData {keys [
      {arriveTangent=-0.01, interpMode=RCIM_Cubic, time=0, value=1}
      {arriveTangent=-0.0035, interpMode=RCIM_Cubic, leaveTangent=-0.0035, tangentMode=RCTM_User, time=4.5352, value=0.585}
      {arriveTangent=0.0016, interpMode=RCIM_Cubic, leaveTangent=0.0016, tangentMode=RCTM_User, time=7.591, value=0.8382}
      {arriveTangent=0.0008, interpMode=RCIM_Cubic, leaveTangent=0.0008, tangentMode=RCTM_User, time=11.1001, value=0.5985}
      {arriveTangent=0.0016, interpMode=RCIM_Cubic, leaveTangent=0.0016, tangentMode=RCTM_User, time=12.9699, value=0.5841}
      {arriveTangent=0.0009, interpMode=RCIM_Cubic, leaveTangent=0.0009, time=15.1685, value=0.7207}
      {arriveTangent=-0.0011, interpMode=RCIM_Cubic, leaveTangent=-0.0011, tangentMode=RCTM_User, time=18.5792, value=0.5891}
      {arriveTangent=0.0002, interpMode=RCIM_Cubic, leaveTangent=0.0002, time=24.1334, value=0.8568}
      {arriveTangent=0.0016, interpMode=RCIM_Cubic, leaveTangent=0.0016, tangentMode=RCTM_User, time=28.0265, value=0.5913}
      {arriveTangent=-0.0003, interpMode=RCIM_Cubic, leaveTangent=-0.0003, time=31.406, value=0.7353}
      {arriveTangent=0.0009, interpMode=RCIM_Cubic, leaveTangent=0.0009, tangentMode=RCTM_User, time=34.9151, value=0.5893}
      {arriveTangent=-0.0008, interpMode=RCIM_Cubic, leaveTangent=-0.0008, time=38.1075, value=0.8568}
      <at 42.0006 0.5839>
      {arriveTangent=0.0024, interpMode=RCIM_Cubic, leaveTangent=0.0024, tangentMode=RCTM_User, time=44.5592, value=0.5891}
      {arriveTangent=-0.0006, interpMode=RCIM_Cubic, leaveTangent=-0.0006, time=48.1356, value=0.8382}
      {arriveTangent=-0.0056, interpMode=RCIM_Cubic, leaveTangent=-0.0056, tangentMode=RCTM_User, time=53.1267, value=0.5841}
      <at 55.5199 0.7454>
      {arriveTangent=-0.0003, interpMode=RCIM_Cubic, leaveTangent=-0.0003, tangentMode=RCTM_User, time=57.746, value=0.5838}
      <at 60 1 RCIM_Cubic>
     ]}}
  }
 }
 {
  id=Enveloper_4
  children [Enveloper_3]
  properties {
   bLoop=true
   bLoopIndefinitely=true
   loopEnd=40
   pitchCurve {editorCurveData {keys [{time=0}]}}
   volumeCurve {editorCurveData {keys [
      {arriveTangent=-0.01, interpMode=RCIM_Cubic, time=0, value=1}
      <at 10 0.7 RCIM_Cubic>
      {arriveTangent=-0.0173, interpMode=RCIM_Cubic, leaveTangent=-0.0173, time=20, value=1}
      {arriveTangent=-0.0749, interpMode=RCIM_Cubic, leaveTangent=-0.0749, tangentMode=RCTM_User, time=25.9279, value=0.4252}
      {arriveTangent=0.129, interpMode=RCIM_Cubic, leaveTangent=0.129, tangentMode=RCTM_User, time=36.1629, value=0.5685}
      <at 40 1 RCIM_Cubic>
     ]}}
  }
 }
 {id=Looping_1, children [Modulator_1]}
 {id=Mixer_0, children [Looping_1,Delay_0], properties {inputVolume [0.8,1]}}
 {id=Modulator_1, children [Delay_2], properties {pitchMax=1.2, pitchMin=1, volumeMax=1, volumeMin=1}}
 {id=QualityLevel_1, children [Mixer_0,Mixer_0,Mixer_0,Mixer_0]}
 {
  id=Random_2
  children [
   WavePlayer_1
   WavePlayer_2
   WavePlayer_3
   WavePlayer_4
   WavePlayer_5
   WavePlayer_6
   WavePlayer_7
   WavePlayer_8
   WavePlayer_9
   WavePlayer_10
   WavePlayer_11
  ]
  properties {weights [1,1,1,1,1,1,1,1,1,1,1]}
 }
 {id=WavePlayer_0, properties {bLooping=true, soundWaveAssetPtr=/Game/Athena/Sounds/Ambient/PirateCove/Amb_PirateCove_SailFlap_01}}
 {id=WavePlayer_1, properties {soundWaveAssetPtr=/Game/Sounds/Artemis/POI/Chrome_BugleBay/Spot_PirateCove_SailFlap_Large_01}}
 {id=WavePlayer_10, properties {soundWaveAssetPtr=/Game/Sounds/Artemis/POI/Chrome_BugleBay/Spot_PirateCove_SailFlap_Large_10}}
 {id=WavePlayer_11, properties {soundWaveAssetPtr=/Game/Sounds/Artemis/POI/Chrome_BugleBay/Spot_PirateCove_SailFlap_Large_11}}
 {id=WavePlayer_2, properties {soundWaveAssetPtr=/Game/Sounds/Artemis/POI/Chrome_BugleBay/Spot_PirateCove_SailFlap_Large_02}}
 {id=WavePlayer_3, properties {soundWaveAssetPtr=/Game/Sounds/Artemis/POI/Chrome_BugleBay/Spot_PirateCove_SailFlap_Large_03}}
 {id=WavePlayer_4, properties {soundWaveAssetPtr=/Game/Sounds/Artemis/POI/Chrome_BugleBay/Spot_PirateCove_SailFlap_Large_04}}
 {id=WavePlayer_5, properties {soundWaveAssetPtr=/Game/Sounds/Artemis/POI/Chrome_BugleBay/Spot_PirateCove_SailFlap_Large_05}}
 {id=WavePlayer_6, properties {soundWaveAssetPtr=/Game/Sounds/Artemis/POI/Chrome_BugleBay/Spot_PirateCove_SailFlap_Large_06}}
 {id=WavePlayer_7, properties {soundWaveAssetPtr=/Game/Sounds/Artemis/POI/Chrome_BugleBay/Spot_PirateCove_SailFlap_Large_07}}
 {id=WavePlayer_8, properties {soundWaveAssetPtr=/Game/Sounds/Artemis/POI/Chrome_BugleBay/Spot_PirateCove_SailFlap_Large_08}}
 {id=WavePlayer_9, properties {soundWaveAssetPtr=/Game/Sounds/Artemis/POI/Chrome_BugleBay/Spot_PirateCove_SailFlap_Large_09}}
]
properties {
 bOverrideAttenuation=true
 maxDistance=6000
 soundClassObject=&Fort_Ambient_Outdoors_Athena
 volumeMultiplier=0.501
 attenuationOverrides {
  attenuationShape=Capsule
  bAttenuateWithLPF=true
  bEnableFocusInterpolation=true
  bEnableListenerFocus=true
  bEnableOcclusion=true
  falloffDistance=5000
  focusAttackInterpSpeed=0.1
  focusReleaseInterpSpeed=0.1
  hPFFrequencyAtMax=500
  lPFRadiusMin=1500
  manualReverbSendLevel=10
  nonFocusVolumeAttenuation=0.5
  occlusionLowPassFilterFrequency=3000
  occlusionTraceChannel=ECC_GameTraceChannel11
  reverbDistanceMax=2000
  reverbSendMethod=Manual
  reverbWetLevelMax=5
  attenuationShapeExtents=<1000 500>
 }
 soundSubmixSends [{
   disableManualSendClamp=true
   maxSendDistance=5000
   maxSendLevel=1.2
   minSendDistance=1500
   minSendLevel=0.4
   sendLevel=1.2
   soundSubmix=&Fortnite_Reverb_Default
   customSendLevelCurve {externalCurve=/Game/Sounds/SubMixes/Reverb/Fortnite_Reverb_Distant_Send}
  }]
}
