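# Sound cue for a looping ambient spot of a small flapping sail at the Shipwreck POI.
# One looping pirate-cove sail wave passes through a mixer shared by all quality levels, then three nested looping envelopers, behind a randomized start delay.
# The envelopers layer pitch drift and two volume curves of different loop lengths to vary the repeating sample.
# Overridden capsule attenuation adds occlusion, focus interpolation and manual reverb send on the outdoor ambient sound class.
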SoundCue /Game/Sounds/Helios/POI/Shipwreck/Spots/Spot_Shipwreck_Sail_Flapping_Smallest_Loop_02_Cue
root=Delay_0
nodes [
 {id=Delay_0, children [Enveloper_4], properties {delayMax=0.75, delayMin=0.2}}
 {
  id=Enveloper_0
  children [QualityLevel_0]
  properties {
   bLoop=true
   bLoopIndefinitely=true
   loopEnd=60
   pitchCurve {editorCurveData {keys [
      {interpMode=RCIM_Cubic, time=0}
      <at 15 0.85 RCIM_Cubic>
      {arriveTangent=0.0091, interpMode=RCIM_Cubic, leaveTangent=0.0091, tangentMode=RCTM_User, time=30}
      <at 45 1.1 RCIM_Cubic>
      {time=60}
     ]}}
   volumeCurve {editorCurveData {keys [<at 0 1>]}}
  }
 }
 {
  id=Enveloper_3
  children [Enveloper_0]
  properties {
   bLoop=true
   bLoopIndefinitely=true
   loopEnd=60
   pitchCurve {editorCurveData {keys [{time=0}]}}
   volumeCurve {editorCurveData {keys [
      {arriveTangent=-0.01, interpMode=RCIM_Cubic, time=0, value=1}
      {arriveTangent=-0.0035, interpMode=RCIM_Cubic, leaveTangent=-0.0035, tangentMode=RCTM_User, time=3.6964, value=0.7711}
      {arriveTangent=0.0016, interpMode=RCIM_Cubic, leaveTangent=0.0016, tangentMode=RCTM_User, time=7.2602, value=0.9432}
      {arriveTangent=0.0008, interpMode=RCIM_Cubic, leaveTangent=0.0008, tangentMode=RCTM_User, time=10.2613, value=0.7846}
      {arriveTangent=0.0016, interpMode=RCIM_Cubic, leaveTangent=0.0016, tangentMode=RCTM_User, time=12.1311, value=0.7702}
      {arriveTangent=0.0009, interpMode=RCIM_Cubic, leaveTangent=0.0009, time=14.8377, value=0.8256}
      {arriveTangent=-0.0011, interpMode=RCIM_Cubic, leaveTangent=-0.0011, tangentMode=RCTM_User, time=17.7404, value=0.7752}
      {arriveTangent=0.0002, interpMode=RCIM_Cubic, leaveTangent=0.0002, time=23.8026, value=0.9617}
      {arriveTangent=0.0016, interpMode=RCIM_Cubic, leaveTangent=0.0016, tangentMode=RCTM_User, time=27.1877, value=0.7774}
      {arriveTangent=-0.0003, interpMode=RCIM_Cubic, leaveTangent=-0.0003, time=31.0752, value=0.8403}
      {arriveTangent=0.0009, interpMode=RCIM_Cubic, leaveTangent=0.0009, tangentMode=RCTM_User, time=34.0763, value=0.7754}
      {arriveTangent=-0.0008, interpMode=RCIM_Cubic, leaveTangent=-0.0008, time=37.7767, value=0.9617}
      <at 41.1618 0.7701>
      {arriveTangent=0.0024, interpMode=RCIM_Cubic, leaveTangent=0.0024, tangentMode=RCTM_User, time=43.7204, value=0.7752}
      {arriveTangent=-0.0006, interpMode=RCIM_Cubic, leaveTangent=-0.0006, time=47.8047, value=0.9432}
      {arriveTangent=-0.0056, interpMode=RCIM_Cubic, leaveTangent=-0.0056, tangentMode=RCTM_User, time=52.2879, value=0.7702}
      <at 55.1891 0.8504>
      {arriveTangent=-0.0003, interpMode=RCIM_Cubic, leaveTangent=-0.0003, tangentMode=RCTM_User, time=56.9072, value=0.7699}
      <at 60 1 RCIM_Cubic>
     ]}}
  }
 }
 {
  id=Enveloper_4
  children [Enveloper_3]
  properties {
   bLoop=true
   bLoopIndefinitely=true
   loopEnd=40
   pitchCurve {editorCurveData {keys [{time=0}]}}
   volumeCurve {editorCurveData {keys [
      {arriveTangent=-0.01, interpMode=RCIM_Cubic, time=0, value=1}
      {arriveTangent=-0.0001, interpMode=RCIM_Cubic, leaveTangent=-0.0001, time=10, value=0.7}
      {arriveTangent=0.0197, interpMode=RCIM_Cubic, leaveTangent=0.0197, time=16.3879, value=0.9979}
      {arriveTangent=-0.0749, interpMode=RCIM_Cubic, leaveTangent=-0.0749, tangentMode=RCTM_User, time=25.4062, value=1.0035}
      {arriveTangent=0.129, interpMode=RCIM_Cubic, leaveTangent=0.129, tangentMode=RCTM_User, time=35.6412, value=1.1468}
      <at 40 1 RCIM_Cubic>
     ]}}
  }
 }
 {id=Mixer_0, children [WavePlayer_0,null], properties {inputVolume [1,1]}}
 {id=QualityLevel_0, children [Mixer_0,Mixer_0,Mixer_0,Mixer_0]}
 {id=WavePlayer_0, properties {bLooping=true, soundWaveAssetPtr=/Game/Athena/Sounds/Ambient/PirateCove/Amb_PirateCove_SailFlap_01}}
]
properties {
 bOverrideAttenuation=true
 maxDistance=1629.0496
 soundClassObject=&Fort_Ambient_Outdoors_Athena
 volumeMultiplier=0.501
 attenuationOverrides {
  attenuationShape=Capsule
  bEnableFocusInterpolation=true
  bEnableOcclusion=true
  falloffDistance=981.7961
  focusAttackInterpSpeed=3
  focusReleaseInterpSpeed=3
  manualReverbSendLevel=10
  nonFocusVolumeAttenuation=0.5
  occlusionLowPassFilterFrequency=3000
  occlusionTraceChannel=ECC_GameTraceChannel11
  reverbDistanceMax=2000
  reverbSendMethod=Manual
  reverbWetLevelMax=5
  attenuationShapeExtents=<647.2534 246.6023>
 }
}
