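# Sound cue for deep, menacing wood creak one-shots at the Shipwreck POI, routed to the indoor ambient sound class.
# A quality-level node picks one random pool of fifteen creaks for the first two levels and a smaller five-creak pool for the last two.
# A root modulator varies pitch on each play.
# It references a shared distant-creak attenuation asset and overrides concurrency to limit simultaneous voices.
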SoundCue /Game/Sounds/Helios/POI/Shipwreck/Spots/Spot_Wood_Creak_Deep_Huge_Menacing_Int_Cue
root=Modulator_0
nodes [
 {id=Modulator_0, children [QualityLevel_0], properties {pitchMax=1.1, pitchMin=0.9}}
 {id=QualityLevel_0, children [Random_2,Random_2,Random_3,Random_3]}
 {
  id=Random_2
  children [
   WavePlayer_11
   WavePlayer_12
   WavePlayer_17
   WavePlayer_18
   WavePlayer_19
   WavePlayer_20
   WavePlayer_21
   WavePlayer_22
   WavePlayer_23
   WavePlayer_24
   WavePlayer_25
   WavePlayer_26
   WavePlayer_27
   WavePlayer_28
   WavePlayer_29
  ]
  properties {weights [1,1,1,1,1,1,1,1,1,1,1,1,1,1,1]}
 }
 {id=Random_3, children [WavePlayer_30, WavePlayer_31, WavePlayer_32, WavePlayer_33, WavePlayer_34], properties {weights [1,1,1,1,1]}}
 {id=WavePlayer_11, properties {soundWaveAssetPtr=&/Spot_Wood_Creak_Deep_Huge_Menacing_01}}
 {id=WavePlayer_12, properties {soundWaveAssetPtr=&/Spot_Wood_Creak_Deep_Huge_Menacing_02}}
 {id=WavePlayer_17, properties {soundWaveAssetPtr=&/Spot_Wood_Creak_Deep_Huge_Menacing_03}}
 {id=WavePlayer_18, properties {soundWaveAssetPtr=&/Spot_Wood_Creak_Deep_Huge_Menacing_04}}
 {id=WavePlayer_19, properties {soundWaveAssetPtr=&/Spot_Wood_Creak_Deep_Huge_Menacing_05}}
 {id=WavePlayer_20, properties {soundWaveAssetPtr=&/Spot_Wood_Creak_Deep_Huge_Menacing_06}}
 {id=WavePlayer_21, properties {soundWaveAssetPtr=&/Spot_Wood_Creak_Deep_Huge_Menacing_07}}
 {id=WavePlayer_22, properties {soundWaveAssetPtr=&/Spot_Wood_Creak_Deep_Huge_Menacing_08}}
 {id=WavePlayer_23, properties {soundWaveAssetPtr=&/Spot_Wood_Creak_Deep_Huge_Menacing_09}}
 {id=WavePlayer_24, properties {soundWaveAssetPtr=&/Spot_Wood_Creak_Deep_Huge_Menacing_10}}
 {id=WavePlayer_25, properties {soundWaveAssetPtr=&/Spot_Wood_Creak_Deep_Huge_Menacing_11}}
 {id=WavePlayer_26, properties {soundWaveAssetPtr=&/Spot_Wood_Creak_Deep_Huge_Menacing_12}}
 {id=WavePlayer_27, properties {soundWaveAssetPtr=&/Spot_Wood_Creak_Deep_Huge_Menacing_13}}
 {id=WavePlayer_28, properties {soundWaveAssetPtr=&/Spot_Wood_Creak_Deep_Huge_Menacing_14}}
 {id=WavePlayer_29, properties {soundWaveAssetPtr=&/Spot_Wood_Creak_Deep_Huge_Menacing_15}}
 {id=WavePlayer_30, properties {soundWaveAssetPtr=&Spot_Wood_Creak_Deep_Huge_Menacing_09}}
 {id=WavePlayer_31, properties {soundWaveAssetPtr=&Spot_Wood_Creak_Deep_Huge_Menacing_10}}
 {id=WavePlayer_32, properties {soundWaveAssetPtr=&Spot_Wood_Creak_Deep_Huge_Menacing_11}}
 {id=WavePlayer_33, properties {soundWaveAssetPtr=&Spot_Wood_Creak_Deep_Huge_Menacing_06}}
 {id=WavePlayer_34, properties {soundWaveAssetPtr=&Spot_Wood_Creak_Deep_Huge_Menacing_01}}
]
properties {
 attenuationSettings=/Game/Sounds/Attenuation/Fort_TileAmbient_OneShot_Distant_Creak_Att
 bOverrideConcurrency=true
 maxDistance=8200
 soundClassObject=/Game/Sounds/FortSoundClassesAndModes/Fort_Ambient_OneShots_Indoors
 volumeMultiplier=0.501
 attenuationOverrides {
  attenuationShape=Capsule
  bAttenuate=false
  bEnableFocusInterpolation=true
  bEnableListenerFocus=true
  distanceAlgorithm=Custom
  falloffDistance=3760
  focusAzimuth=25
  focusReleaseInterpSpeed=0.4
  nonFocusAzimuth=55
  nonFocusVolumeAttenuation=0.4
  nonSpatializedRadiusEnd=1500
  nonSpatializedRadiusStart=2000
  stereoSpread=100
  attenuationShapeExtents {x=2350}
  customAttenuationCurve {editorCurveData {keys [
     <at 0.2 0>
     <at 0.35 1>
     {arriveTangent=-0.8696, interpMode=RCIM_Cubic, leaveTangent=-0.8696, time=0.8, value=1}
     {arriveTangent=-0.0046, interpMode=RCIM_Cubic, leaveTangent=-0.0046, tangentMode=RCTM_User, time=1.5, value=0}
    ]}}
 }
 concurrencyOverrides {maxCount=2, voiceStealReleaseTime=2}
}
