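# Looping shipwreck spot ambience of water hitting a wooden hull, limited to one instance.
# Mixes randomly chosen impact one-shots, repeated after 2–4 second delays, with a pitched-down loop bed shaped by nested looping pitch and volume envelopes.
# The mix feeds only the first two quality slots (the others are empty), then a random pitch modulator.
# Overrides add cone attenuation with low-pass filtering and listener focus, plus a reverb send.
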SoundCue /Game/Sounds/Helios/POI/Shipwreck/Spots/Spot_Wood_Hull_Water_Impact_Loop_Box_Int_02_Cue
root=Modulator_0
nodes [
 {id=Delay_0, children [Random_2], properties {delayMax=4, delayMin=2}}
 {
  id=Enveloper_0
  children [WavePlayer_11]
  properties {
   bLoop=true
   bLoopIndefinitely=true
   loopEnd=30
   pitchCurve {editorCurveData {keys [{interpMode=RCIM_Cubic, time=0}, <at 15.9468 0.9181 RCIM_Cubic>, {interpMode=RCIM_Cubic, time=30}]}}
   volumeCurve {editorCurveData {keys [<at 0 1>]}}
  }
 }
 {
  id=Enveloper_1
  children [Enveloper_0]
  properties {
   bLoop=true
   bLoopIndefinitely=true
   loopEnd=60
   pitchCurve {editorCurveData {keys [{time=0}]}}
   volumeCurve {editorCurveData {keys [
      {arriveTangent=-0.01, interpMode=RCIM_Cubic, time=0, value=1}
      <at 15 0.6 RCIM_Cubic>
      {arriveTangent=0.0001, interpMode=RCIM_Cubic, leaveTangent=0.0001, tangentMode=RCTM_User, time=30, value=1}
      <at 45 0.4 RCIM_Cubic>
      <at 60 1 RCIM_Cubic>
     ]}}
  }
 }
 {id=Looping_0, children [Delay_0]}
 {id=Mixer_0, children [Looping_0,Modulator_1], properties {inputVolume [0.4,0.6]}}
 {id=Modulator_0, children [QualityLevel_0], properties {pitchMax=1.1, pitchMin=0.8}}
 {id=Modulator_1, children [Enveloper_1], properties {pitchMax=0.8, pitchMin=0.8}}
 {id=QualityLevel_0, children [Mixer_0,Mixer_0,null,null]}
 {id=Random_2, children [WavePlayer_8, WavePlayer_9, WavePlayer_10], properties {weights [1,1,1]}}
 {id=WavePlayer_10, properties {soundWaveAssetPtr=&/Spot_Wood_Hull_Water_Impact_07}}
 {id=WavePlayer_11, properties {bLooping=true, soundWaveAssetPtr=&/Spot_Wood_Hull_Water_Impact_Loop_02}}
 {id=WavePlayer_8, properties {soundWaveAssetPtr=&/Spot_Wood_Hull_Water_Impact_05}}
 {id=WavePlayer_9, properties {soundWaveAssetPtr=&/Spot_Wood_Hull_Water_Impact_06}}
]
properties {
 bOverrideAttenuation=true
 bOverrideConcurrency=true
 maxDistance=1000
 soundClassObject=/Game/Sounds/FortSoundClassesAndModes/Fort_Earthquakes
 volumeMultiplier=0.4
 attenuationOverrides {
  attenuationShape=Cone
  bAttenuateWithLPF=true
  bEnableFocusInterpolation=true
  bEnableListenerFocus=true
  falloffDistance=650
  focusAttackInterpSpeed=0.9
  focusReleaseInterpSpeed=0.4
  hPFFrequencyAtMax=100
  lPFFrequencyAtMax=1000
  lPFRadiusMax=1000
  lPFRadiusMin=500
  nonFocusVolumeAttenuation=0.4
  stereoSpread=100
  attenuationShapeExtents=<350 80 10>
 }
 concurrencyOverrides {maxCount=1, voiceStealReleaseTime=3}
 soundSubmixSends [{minSendLevel=0.6, soundSubmix=&Fortnite_Reverb_Default}]
}
