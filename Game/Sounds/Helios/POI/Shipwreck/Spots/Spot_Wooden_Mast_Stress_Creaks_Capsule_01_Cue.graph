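# Looping ambient Sound Cue of stressed wooden mast creaks for the Shipwreck POI, reusing Artemis wooden bridge recordings.
# Every 6–18 seconds it plays a random creak with varied pitch and volume, shaped by a fading envelope.
# A quality switch gives the first two levels five variants and the last two only two.
# Overrides set capsule attenuation with distance filtering and listener focus, cap concurrency at two, and send to an exterior ambient bus.
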SoundCue /Game/Sounds/Helios/POI/Shipwreck/Spots/Spot_Wooden_Mast_Stress_Creaks_Capsule_01_Cue
root=Looping_0
nodes [
 {id=Delay_0, children [QualityLevel_0], properties {delayMax=18, delayMin=6}}
 {
  id=Enveloper_1
  children [Modulator_0]
  properties {pitchCurve {editorCurveData {keys [{time=0}]}}, volumeCurve {editorCurveData {keys [<at 0 1>,<at 2 1>,<at 4.0204 0.5>,<at 7.0204 -0.0282>]}}}
 }
 {
  id=Enveloper_2
  children [Modulator_1]
  properties {pitchCurve {editorCurveData {keys [{time=0}]}}, volumeCurve {editorCurveData {keys [<at 0 1>,<at 2 1>,<at 4.0204 0.5>,<at 7.0204 -0.0282>]}}}
 }
 {id=Looping_0, children [Delay_0]}
 {id=Modulator_0, children [Random_0], properties {pitchMin=0.8, volumeMin=0.8}}
 {id=Modulator_1, children [Random_1], properties {pitchMin=0.8, volumeMin=0.8}}
 {id=QualityLevel_0, children [Enveloper_1,Enveloper_1,Enveloper_2,Enveloper_2]}
 {id=Random_0, children [WavePlayer_0,WavePlayer_1,WavePlayer_2,WavePlayer_3,WavePlayer_4], properties {weights [1,1,1,1,1]}}
 {id=Random_1, children [WavePlayer_11, WavePlayer_12], properties {weights [1,1]}}
 {
  id=WavePlayer_0
  properties {soundWaveAssetPtr=/Game/Sounds/Artemis/POI/Normal_Bridges/Bridges_Wooden_Corridor/Spots/Spot_Bridge_Wooden_Corridor_Creaks_09}
 }
 {
  id=WavePlayer_1
  properties {soundWaveAssetPtr=/Game/Sounds/Artemis/POI/Normal_Bridges/Bridges_Wooden_Corridor/Spots/Spot_Bridge_Wooden_Corridor_Creaks_01}
 }
 {
  id=WavePlayer_11
  properties {soundWaveAssetPtr=/Game/Sounds/Artemis/POI/Normal_Bridges/Bridges_Wooden_Corridor/Spots/Spot_Bridge_Wooden_Corridor_Creaks_02}
 }
 {
  id=WavePlayer_12
  properties {soundWaveAssetPtr=/Game/Sounds/Artemis/POI/Normal_Bridges/Bridges_Wooden_Corridor/Spots/Spot_Bridge_Wooden_Corridor_Creaks_03}
 }
 {
  id=WavePlayer_2
  properties {soundWaveAssetPtr=&Spot_Bridge_Wooden_Corridor_Creaks_02}
 }
 {
  id=WavePlayer_3
  properties {soundWaveAssetPtr=&Spot_Bridge_Wooden_Corridor_Creaks_03}
 }
 {
  id=WavePlayer_4
  properties {soundWaveAssetPtr=/Game/Sounds/Artemis/POI/Normal_Bridges/Bridges_Wooden_Corridor/Spots/Spot_Bridge_Wooden_Corridor_Creaks_04}
 }
]
properties {
 attenuationSettings=/Game/Sounds/Attenuation/Fort_TileAmbient_OneShot_Distant_Creak_Att
 bOverrideAttenuation=true
 bOverrideConcurrency=true
 maxDistance=3643.4309
 soundClassObject=&Fort_Ambient_Outdoors_Athena
 volumeMultiplier=0.451
 attenuationOverrides {
  attenuationShape=Capsule
  bAttenuateWithLPF=true
  bEnableFocusInterpolation=true
  bEnableListenerFocus=true
  falloffDistance=2548.4529
  focusReleaseInterpSpeed=0.9
  hPFFrequencyAtMax=250
  lPFFrequencyAtMax=1500
  lPFRadiusMax=4000
  lPFRadiusMin=1500
  nonFocusVolumeAttenuation=0.75
  attenuationShapeExtents=<1094.978 350>
 }
 busSends [{
   audioBus=null
   maxSendDistance=4000
   minSendDistance=1500
   minSendLevel=0.4
   sendLevel=1
   soundSourceBus=&AmbientAudio_Exterior_SourceBus
   customSendLevelCurve {externalCurve=/Game/Sounds/SubMixes/Reverb/Fortnite_Reverb_Distant_Send}
  }]
 concurrencyOverrides {maxCount=2, resolutionRule=PreventNew}
}
