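# Sound cue for looping, quiet indoor ambient wooden mast creaks at the Helios shipwreck POI, reusing Artemis wooden-bridge creak waves.
# A randomized delay spaces each repeat before a quality switch.
# The first two quality slots mix two random picks from five creaks; the last two play one of two.
# Overridden capsule attenuation adds low-pass filtering and listener focus, and concurrency caps instances by preventing new ones.
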SoundCue /Game/Sounds/Helios/POI/Shipwreck/Spots/Spot_Wooden_Mast_Stress_Creaks_Capsule_Int_Cue
root=Looping_0
nodes [
 {id=Delay_0, children [QualityLevel_0], properties {delayMax=18, delayMin=4}}
 {id=Looping_0, children [Delay_0]}
 {id=Mixer_0, children [Random_2,Random_0], properties {inputVolume [1,1]}}
 {id=QualityLevel_0, children [Mixer_0,Mixer_0,Random_1,Random_1]}
 {id=Random_0, children [WavePlayer_0,WavePlayer_1,WavePlayer_2,WavePlayer_3,WavePlayer_4], properties {weights [1,1,1,1,1]}}
 {id=Random_1, children [WavePlayer_11, WavePlayer_12], properties {weights [1,1]}}
 {id=Random_2, children [WavePlayer_5,WavePlayer_6,WavePlayer_7,WavePlayer_8,WavePlayer_9], properties {weights [1,1,1,1,1]}}
 {
  id=WavePlayer_0
  properties {soundWaveAssetPtr=/Game/Sounds/Artemis/POI/Normal_Bridges/Bridges_Wooden_Corridor/Spots/Spot_Bridge_Wooden_Corridor_Creaks_09}
 }
 {
  id=WavePlayer_1
  properties {soundWaveAssetPtr=/Game/Sounds/Artemis/POI/Normal_Bridges/Bridges_Wooden_Corridor/Spots/Spot_Bridge_Wooden_Corridor_Creaks_01}
 }
 {
  id=WavePlayer_11
  properties {soundWaveAssetPtr=/Game/Sounds/Artemis/POI/Normal_Bridges/Bridges_Wooden_Corridor/Spots/Spot_Bridge_Wooden_Corridor_Creaks_02}
 }
 {
  id=WavePlayer_12
  properties {soundWaveAssetPtr=/Game/Sounds/Artemis/POI/Normal_Bridges/Bridges_Wooden_Corridor/Spots/Spot_Bridge_Wooden_Corridor_Creaks_03}
 }
 {
  id=WavePlayer_2
  properties {soundWaveAssetPtr=&Spot_Bridge_Wooden_Corridor_Creaks_02}
 }
 {
  id=WavePlayer_3
  properties {soundWaveAssetPtr=&Spot_Bridge_Wooden_Corridor_Creaks_03}
 }
 {
  id=WavePlayer_4
  properties {soundWaveAssetPtr=/Game/Sounds/Artemis/POI/Normal_Bridges/Bridges_Wooden_Corridor/Spots/Spot_Bridge_Wooden_Corridor_Creaks_04}
 }
 {
  id=WavePlayer_5
  properties {soundWaveAssetPtr=&Spot_Bridge_Wooden_Corridor_Creaks_09}
 }
 {
  id=WavePlayer_6
  properties {soundWaveAssetPtr=&Spot_Bridge_Wooden_Corridor_Creaks_01}
 }
 {
  id=WavePlayer_7
  properties {soundWaveAssetPtr=&Spot_Bridge_Wooden_Corridor_Creaks_02}
 }
 {
  id=WavePlayer_8
  properties {soundWaveAssetPtr=&Spot_Bridge_Wooden_Corridor_Creaks_03}
 }
 {
  id=WavePlayer_9
  properties {soundWaveAssetPtr=&Spot_Bridge_Wooden_Corridor_Creaks_04}
 }
]
properties {
 attenuationSettings=/Game/Sounds/Attenuation/Fort_TileAmbient_OneShot_Distant_Creak_Att
 bEnableSubmixSends=false
 bOverrideAttenuation=true
 bOverrideConcurrency=true
 maxDistance=700
 soundClassObject=/Game/Sounds/FortSoundClassesAndModes/Fort_Ambient_OneShots_Indoors
 volumeMultiplier=0.15
 attenuationOverrides {
  attenuationShape=Capsule
  bAttenuateWithLPF=true
  bEnableFocusInterpolation=true
  bEnableListenerFocus=true
  falloffDistance=500
  focusReleaseInterpSpeed=0.9
  hPFFrequencyAtMax=250
  lPFRadiusMax=1000
  lPFRadiusMin=250
  nonFocusVolumeAttenuation=0.75
  attenuationShapeExtents {x=200}
 }
 concurrencyOverrides {maxCount=2, resolutionRule=PreventNew}
}
